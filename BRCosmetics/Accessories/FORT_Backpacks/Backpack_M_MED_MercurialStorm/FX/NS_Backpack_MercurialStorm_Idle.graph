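# Idle Niagara effect for the Mercurial Storm backpack cosmetic, adding a blue glow and light at a small offset.
# A single LightAndGlow emitter burst-spawns particles that persist until deactivation, drawn by three sprite renderers and a light renderer.
# Module-namespace sprite attributes and light attributes supply separate colors and sizes, and time-of-day scalars adjust sprite brightness.
# User color parameters are exposed, but no shown module input reads them.
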NiagaraSystem /BRCosmetics/Accessories/FORT_Backpacks/Backpack_M_MED_MercurialStorm/FX/NS_Backpack_MercurialStorm_Idle
dependencies {usedDataInterfaces [], usedDynamicInputs [], usedModules [&SystemState]}
emitters {LightAndGlow {
  emitterUpdateScript [
   &EmitterState(Life Cycle Mode=::NewEnumerator1, Loop Behavior=::NewEnumerator1, Loop Duration Mode=::NewEnumerator1, UseLoopDelay=false)
   &SpawnBurst_Instantaneous(Use Spawn Probability=false, UseLoopCountLimit=false)
  ]
  particleSpawnScript [
   &InitializeParticle(
    Color=<rgba 0 0.2172 2 0.5>
    Position Offset=<0 -5 -4>
    Sprite Size Mode=::NewEnumerator3
    Uniform Sprite Size=15
    UsePositionOffset=true
    Write Material Random=true
   )
   &FN_Sprite_Attributes(
    Color=<rgba 0 0.5512 1 0.5>
    Create Color Attribute=true
    Create SpriteSize Attribute=true
    SpriteSize=<30 30>
    Write to Module Namespace=true
   )
   &CameraOffset(Camera Offset Amount=4)
   &DynamicMaterialParameters(
    Index 0 Param 1=7
    Module.Param0WriteEnabled=true
    Module.Param1WriteEnabled=true
    Module.Param2WriteEnabled=true
    Module.Param3WriteEnabled=true
   )
  ]
  particleUpdateScript [
   &ParticleState(Kill Particles When Lifetime Has Elapsed=false, Let Infinitely Lived Particles Die When Emitter Deactivates=true)
   &FN_Light_Attributes(
    Color=<rgba 0 0.1751 0.8142 1>
    Create Color Attribute=true
    Create Exponent Attribute=true
    Create Position Attribute=true
    Create Radius Attribute=true
    Exponent=6
    Position=<0 -5 -7>
    Radius=15
    Scale Color=6
   )
   &FN_Sprite_Attributes(
    Color=&MultiplyLinearColor(
     Linear Color=<rgba 0 2.205 4 0.5>
     Scale Factor=&MakeVector4_FromFloat(Float=&FNTimeOfDayScalar(Night Scale=0.3, Morning Scale=0.5, Evening Scale=0.7))
    )
    Create Color Attribute=true
    Create SpriteSize Attribute=true
    SpriteSize=<8 8>
    Write to Module Namespace=true
   ) {module=FN_Sprite_Attributes001}
   &ScaleColor(
    Scale RGB=&VectorFromFloat(Value=&FNTimeOfDayScalar(Night Scale=0.3, Morning Scale=0.7, Evening Scale=0.7))
    ScaleA=true
    ScaleRGB=true
   )
  ]
  rendererClasses [&NiagaraSpriteRendererProperties, &NiagaraLightRendererProperties]
  renderers [
   {rendererClass=&NiagaraSpriteRendererProperties}
   {rendererClass=&NiagaraSpriteRendererProperties, rendererIndex=1}
   {rendererClass=&NiagaraLightRendererProperties, rendererIndex=2}
   {rendererClass=&NiagaraSpriteRendererProperties, rendererIndex=3}
  ]
 }}
summary {
 systemName=NS_Backpack_MercurialStorm_Idle
 emitters [{emitterName=LightAndGlow, rendererClasses [&NiagaraSpriteRendererProperties, &NiagaraLightRendererProperties]}]
 userVariables [
  User.ColorA {type=LinearColor, description="", defaultValue {struct=LinearColor, value=<rgba 0 0.0083 0.05 1>}}
  User.ColorB {type=LinearColor, description="", defaultValue {struct=LinearColor, value=<rgba 0 0.2 0.8 1>}}
  User.HDR_Tint {type=LinearColor, description="", defaultValue {struct=LinearColor, value=<rgba 0 0.6667 1 1>}}
  User.M_Background_Color {type=LinearColor, description="", defaultValue {struct=LinearColor, value=<rgba 0.35 0.05 0.05 1>}}
  User.M_Lines_Color {type=LinearColor, description="", defaultValue {struct=LinearColor, value=<rgba 15 0 0 1>}}
 ]
}
